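# Unlit additive material for animated light rays on particles, including Niagara meshes, in the VictoryCrownsAmbience effects.
# Counter-panned ray textures and noise plus a center glow blend with a Ceil-stylized variant, masked by a noise-varied vertical gradient and curve-atlas ramps.
# Particle color, dynamic parameters and switchable color range, hue shift and time-of-day tint drive the look.
# Switchable coin-bob or hit-wobble offsets deform it; depth and near-camera fades soften opacity.
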Material /VictoryCrownsAmbience/Items/Assets/FX/M_Lightrays_ramped
graph {
 properties {
  bUsedWithNiagaraMeshParticles=true
  blendMode=BLEND_Additive
  pixelDepthOffsetMode=PDOM_Legacy
  shadingModel=MSM_Unlit
  usedShadingModels=Unlit
 }
 nodes [
  n1 = ScalarParameter() {defaultValue=1, parameterName=Color_Day}
  n2 = ScalarParameter() {defaultValue=1, parameterName=Color_Night}
  n3 = LinearInterpolate(
   A=VectorParameter() {parameterName=Color_01, defaultValue=<rgba 1 0 0.1327 1>}.RGB
   B=VectorParameter() {parameterName=Color_02, defaultValue=<rgba 0 0 1 1>}.RGB
  )
  n4 = ParticleColor()
  n5 = DynamicParameter() {defaultValue=<rgba 0 1 1 1>, paramNames [Random 0-1, Panner Speed, "U Tile ", Param4]}
  n6 = Multiply(A=TextureCoordinate(), B=@n5."U Tile ")
  n7 = ScalarParameter() {defaultValue=2, parameterName=Gradient Tile A}
  n8 = Multiply(A=@n5.Panner Speed, B=Add(A=@n5.Random 0-1, B=Time()))
  n9 = TextureSample(UVs=Panner(Coordinate=Multiply(A=@n6, B=@n7), Time=@n8) {speedX=0.0702}) {texture=/Game/Effects/Fort_Effects/Textures/Outlander/T_InTheZone_LightRays_01}
  n10 = Multiply(A=@n5."U Tile ", B=TextureCoordinate())
  n11 = ScalarParameter() {defaultValue=2.5, parameterName=Gradient Tile B}
  n12 = TextureSample(UVs=Panner(Coordinate=Multiply(A=@n10, B=@n11), Time=@n8) {speedX=-0.047}) {texture=&T_InTheZone_LightRays_01}
  n13 = TextureCoordinate() {uTiling=0.33, vTiling=0.01}
  n14 = DynamicParameter() {defaultValue=<rgba 0 1 1 1>, paramNames [Random 0-1, Panner Speed, "U Tile ", Param4]}
  n15 = TextureSample(UVs=Panner(Coordinate=Add(A=@n13, B=@n14.Random 0-1), Time=@n8) {speedX=0.0436}) {texture=&T_NoisePack_01}
  n16 = ScalarParameter() {defaultValue=3, parameterName=Mask Brightness}
  n17 = TextureCoordinate()
  n18 = ScalarParameter() {defaultValue=-0.03, parameterName=Slide}
  n19 = OneMinus(None=ComponentMask(None=Add(A=@n17, B=@n18)) {g=true})
  n20 = ScalarParameter() {defaultValue=12, parameterName=Center Exponent}
  n21 = Add(
   A=Multiply(A=Multiply(A=@n9.R, B=@n12.G), B=Multiply(A=@n15.R, B=@n16))
   B=Multiply(A=Power(Base=@n19, Exp=@n20), B=ScalarParameter() {defaultValue=2, parameterName=Center Brightness})
  )
  n22 = Add(A=@n21, B=ScalarParameter() {defaultValue=-0.7, parameterName=Ceil Bias})
  n23 = LinearInterpolate(
   A=@n21
   B=Multiply(A=Ceil(None=@n22), B=ScalarParameter() {defaultValue=3, parameterName=Ceil Brightness})
   Alpha=ScalarParameter() {defaultValue=0.5, parameterName=Style Lerp}
  )
  n24 = TextureCoordinate()
  n25 = Panner(
   Coordinate=TextureCoordinate() {uTiling=3, vTiling=0}
   Speed=AppendVector(
    A=ScalarParameter() {defaultValue=0.1, parameterName=PanSpeedX, sliderMax=0.1}
    B=ScalarParameter() {defaultValue=0.2, parameterName=PanSpeedY, sliderMax=0.1}
   )
  ) {speedX=0.2, speedY=0.1}
  n26 = Power(
   Base=OneMinus(None=ComponentMask(None=@n24) {g=true})
   Exp=LinearInterpolate(Alpha=TextureSample(UVs=@n25) {texture=&T_NoisePack_01}.R) {constA=2, constB=6}
  )
  n27 = TextureCoordinate()
  n28 = CurveAtlasRowParameter(CurveTime=ComponentMask(None=@n27) {r=true}) {
   atlas=/VictoryCrownsAmbience/Items/Assets/FX/Curves/CA_VC_curves
   curve=/VictoryCrownsAmbience/Items/Assets/FX/Curves/C_VC_vRamp_C
   defaultValue=4
   parameterName=uRamp
  }
  n29 = CurveAtlasRowParameter(CurveTime=ComponentMask(None=@n27) {g=true}) {
   atlas=&CA_VC_curves
   curve=/VictoryCrownsAmbience/Items/Assets/FX/Curves/C_VC_vRamp_D
   defaultValue=5
   parameterName=vRamp
  }
  n30 = Multiply(
   A=StaticSwitchParameter(True=Multiply(A=@n3, B=@n4.RGB), False=@n4.RGB) {parameterName=Use_ColorRange_Instead_of_RandomColor?}
   B=Multiply(A=Multiply(A=@n23, B=@n26), B=Multiply(A=@n28.R, B=@n29.R))
  )
  n31 = ScalarParameter() {defaultValue=1.2, parameterName=HueShift_Fixed}
  n32 = Time()
  n33 = ScalarParameter() {defaultValue=1.2, parameterName=HueShiftSpeed}
  n34 = Multiply(A=Frac(None=Multiply(A=@n32, B=@n33) {constB=1.2}), B=ScalarParameter() {defaultValue=1, parameterName=HueShift_Amount})
  n35 = StaticSwitchParameter(
   True=@n30
   False=MaterialFunctionCall(Hue Shift Percentage=StaticSwitchParameter(True=@n31, False=@n34) {parameterName=Use_Fixed_HueShift_Value?}, Texture=@n30) {materialFunction=&HueShift}.Result
  ) {parameterName=NotDiscoLights}
  n36 = StaticSwitchParameter(
   True=@n35
   False=Multiply(
    A=@n35
    B=MaterialFunctionCall(Color_Morning=@n1, Color_Day=@n1, Color_Evening=@n2, Color_Night=@n2) {materialFunction=&Color_Time_of_Day}.Result
   )
  ) {defaultValue=true, parameterName=Color of day}
  MaterialFunctionCall(Emissive=@n36) {materialFunction=&MF_FX_EmissiveAdjustments}
  Reroute(None=@n36)
  n37 = VectorParameter() {bUseCustomPrimitiveData=true, group=z_code, parameterName=WSDeformationCenter, primitiveDataIndex=20}
  n38 = ScalarParameter() {group=z_code, parameterName=Code_DestructionAmount}
  n39 = StaticBoolParameter() {defaultValue=true, group=z_code, parameterName=Code_UseHitEffects}
  n40 = ScalarParameter() {bUseCustomPrimitiveData=true, defaultValue=1, group=z_code, parameterName=WSDamageRadius, primitiveDataIndex=26}
  n41 = ScalarParameter() {bUseCustomPrimitiveData=true, group=z_code, parameterName=DistanceCullFadeAnimationOffset, primitiveDataIndex=12}
  StaticSwitchParameter(
   True=MaterialFunctionCall!(TimeScale, MaxHeight, MinHeight) {materialFunction=/Game/Playgrounds/Items/Props/Materials/MF_Creative_CoinBobble}.WorldPositionOffset
   False=StaticSwitchParameter(
    True=MaterialFunctionCall!(
     WSDeformationCenter=@n37.RGB
     Code_DestructionAmount=@n38
     Code_UseUvForDamage
     Code_UseHitEffects=@n39
     Code_Emissive
     WSDamageRadius=@n40
     CrackDamageShader
     Diffuse
     ScalarModDamageCracks
     ScalarOffsetforDamageCracks
     ExposeLerpFunctionForInvertingTheHitDistanceCalc
     InvertDamageFalloff
     CrackWorldPosition
     WorldPosition
     DistanceCullFadeAnimationOffset=@n41
    ) {materialFunction=/Game/Packages/jl_tests/Fortnite_HitEffects}.Hit WPO
    False=Constant()
   ) {parameterName=Use Hit Wobble}
  ) {parameterName=CreativeMode_CoinBob}
  n42 = MaterialFunctionCall(Distance=ScalarParameter() {defaultValue=250, parameterName=NearCamFade_Distance}) {materialFunction=&Near_Camera_Fade}
  n43 = ScalarParameter() {defaultValue=2, parameterName=NearCamFade_Exponent, sliderMax=5}
  n44 = DepthFade(FadeDistance=ScalarParameter() {defaultValue=50, parameterName=Depth Fade Distance})
  n45 = Constant() {r=1}
  n46 = Constant() {r=1}
  n47 = Multiply(
   A=@n4.A
   B=StaticSwitchParameter(
    True=FeatureLevelSwitch(Default=@n44, ES2=@n45, ES3_1=@n45, SM4_REMOVED=@n44)
    False=QualitySwitch(Default=@n44, Low=@n46, High=@n44, Medium=@n44)
   ) {parameterName=Use_MobilePerf_Settings}
  )
  StaticSwitchParameter(True=Multiply(A=Power(Base=@n42.Result, Exp=@n43), B=@n47), False=@n47) {parameterName=Use_NearCamFade?}
 ]
}
