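# Translucent Niagara sprite and ribbon material for peaked light rays in the Victory Crowns ambience FX.
# U and V curve-atlas ramps shape a mask, multiplied by panning noise, time-offset by a dynamic parameter, and a near-camera fade.
# That mask indexes a color curve, with weight, dynamic ScaleColor and particle-color adjustments.
# It also feeds a depth fade multiplied by particle alpha for soft opacity.
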Material /VictoryCrownsAmbience/Items/Assets/FX/M_VC_PeakedRays
graph {
 properties {
  bUsedWithNiagaraRibbons=true
  bUsedWithNiagaraSprites=true
  blendMode=BLEND_Translucent
  pixelDepthOffsetMode=PDOM_Legacy
  usedShadingModels=Default Lit
 }
 nodes [
  n1 = DynamicParameter()
  Add(A=@n1.Param1)
  n2 = TextureCoordinate()
  n3 = MaterialFunctionCall(Float2=@n2) {materialFunction=&BreakOutFloat2Components}
  n4 = AppendVector(A=Multiply(A=@n3.R) {constB=0.1}, B=Multiply(A=@n3.G) {constB=0.3})
  Add(A=@n4, B=AppendVector(A=@n1.Param1, B=@n1.Param1))
  n5 = DynamicParameter() {parameterIndex=3, paramNames [ScaleColor,Param2,Param3,Param4]}
  n6 = MaterialFunctionCall(Float2=@n2) {materialFunction=&BreakOutFloat2Components}
  n7 = CurveAtlasRowParameter(CurveTime=@n6.R) {
   atlas=/VictoryCrownsAmbience/Items/Assets/FX/Curves/CA_VC_curves
   curve=/VictoryCrownsAmbience/Items/Assets/FX/Curves/C_VC_uRamp_D
   defaultValue=6
   parameterName=Opacity_U
  }
  n8 = CurveAtlasRowParameter(CurveTime=@n6.G) {
   atlas=&CA_VC_curves
   curve=/VictoryCrownsAmbience/Items/Assets/FX/Curves/C_VC_vRamp_F
   defaultValue=7
   parameterName=Opacity_V
  }
  n9 = Add(A=Time(), B=@n1.Param1)
  n10 = Max(A=TextureSample(UVs=Panner(Coordinate=@n4, Time=@n9) {speedY=0.1}) {texture=/Game/Global/Textures/Default/Noise/TankMoteNoise}.R) {constB=0.5}
  n11 = ScalarParameter() {defaultValue=250, parameterName=NearCamFade_Distance, sliderMax=500}
  n12 = Multiply(
   A=Multiply(
    A=Multiply(A=@n7, B=@n8)
    B=MaterialFunctionCall(X=@n10, Y=@n10, Z=@n10) {materialFunction=&MakeFloat3}.Result
   )
   B=Power(
    Base=MaterialFunctionCall(Distance=@n11) {materialFunction=&Near_Camera_Fade}.Result
    Exp=ScalarParameter() {defaultValue=1, parameterName=NearCamFade_Exponent, sliderMax=5}
   )
  )
  n13 = CurveAtlasRowParameter(CurveTime=MaterialFunctionCall(Float3=@n12) {materialFunction=&BreakOutFloat3Components}.R) {
   atlas=&CA_VC_curves
   curve=/VictoryCrownsAmbience/Items/Assets/FX/Curves/CA_VC_colCurve_A
   defaultValue=2
  }
  n14 = Multiply(A=@n13, B=ScalarParameter() {defaultValue=1, parameterName=ColorWeight})
  MaterialFunctionCall(Emissive=Add(A=@n14, B=Multiply(A=@n14, B=@n5.ScaleColor))) {materialFunction=&MF_FX_EmissiveAdjustments}
  n15 = ParticleColor()
  Multiply(A=@n13, B=ComponentMask(None=@n15.RGB) {b=true, g=true, r=true})
  Multiply(A=DepthFade(Opacity=@n12, FadeDistance=ScalarParameter() {defaultValue=12, parameterName=DepthFade_Distance, sliderMax=500}), B=@n15.A)
 ]
}
