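# Niagara body aura for Victory Crowns ambience: orange-gold glows, streaks, sparkles and chest rays following a character's skeleton.
# FN_BoneFlow spreads burst sprites along bone sets, then scratch modules offset them radially, with staggered pulses on streaks.
# Two chest-anchored mesh emitters layer orange and dark purple rays, while User.Flare gates rising curl-noise sparkles.
# FlareAll boosts non-sparkle brightness alongside FXBrightnessScale; time of day dims sprites, and camera distance fades streaks.
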NiagaraSystem /VictoryCrownsAmbience/Items/Assets/FX/NS_VC_body_B
dependencies {
 usedDataInterfaces [
  &NiagaraDataInterfaceColorCurve
  &NiagaraDataInterfaceCurve
  &NiagaraDataInterfaceVector2DCurve
  &NiagaraDataInterfaceArrayInt32
  &NiagaraDataInterfaceArrayFloat3
  &NiagaraDataInterfaceSkeletalMesh
  &NiagaraDataInterfaceCamera
 ]
 usedDynamicInputs [
  &VectorLength
  &ModuloFloat
  &ClampFloat
  /Niagara/DynamicInputs/Camera/CameraFade
 ]
 usedModules [
  &self:SystemSpawnScript.SystemScriptSource.SystemScriptGraph.NiagaraNodeAssignment_5.SetVariables_#1
  &SystemState
  &Timeline
  &self:SystemSpawnScript.SystemScriptSource.SystemScriptGraph.NiagaraNodeAssignment_2.SetVariables_#2
  &self:SystemSpawnScript.SystemScriptSource.SystemScriptGraph.NiagaraNodeAssignment_4.SetVariables_#3
  &self:SampleSingleBoneSpeed
  &self:SystemSpawnScript.SystemScriptSource.SystemScriptGraph.NiagaraNodeAssignment_0.SetVariables_#4
  &self:SystemSpawnScript.SystemScriptSource.SystemScriptGraph.NiagaraNodeAssignment_1.SetVariables_#5
  &self:SystemSpawnScript.SystemScriptSource.SystemScriptGraph.NiagaraNodeAssignment_3.SetVariables_#6
  &self:SystemSpawnScript.SystemScriptSource.SystemScriptGraph.NiagaraNodeAssignment_6.SetVariables_#7
 ]
}
emitters {
 Sparckle_Body {
  emitterUpdateScript [
   &EmitterState()
   &SpawnRate(
    SpawnRate=&MakeCustomFloatFromBool(Bool=&SetBoolByFloatComparison(A=$User.Flare, B=0.2), True Float=50, False Float=0)
    Use Spawn Probability=false
   )
  ]
  particleSpawnScript [
   &InitializeParticle(
    Color Maximum=<rgba 0.9039 1 0.6539 1>
    Color Minimum=<rgba 1 0.3801 0 1>
    Color Mode=::NewEnumerator2
    Lifetime=&RandomRangeFloat(Minimum=0.5, Maximum=1)
    Sprite Rotation Mode=::NewEnumerator2
    Sprite Size Mode=::NewEnumerator1
    Uniform Sprite Size Max=5
    Uniform Sprite Size Min=3
    UsePositionOffset=false
    Write Material Random=true
   )
   /SetVariables(Particles.RandomCore:NiagaraFloat=&FloatFromCurve(
     FloatCurve=DataInterface() {propertyValues {Curve {keys [
         {arriveTangent=0.0559, interpMode=RCIM_Cubic, leaveTangent=0.0559, tangentMode=RCTM_User, time=0, value=0.15}
         {arriveTangent=0.1895, interpMode=RCIM_Cubic, leaveTangent=0.1895, tangentMode=RCTM_User, time=0.9, value=0.3}
         {arriveTangent=1.1808, interpMode=RCIM_Cubic, leaveTangent=1.1808, tangentMode=RCTM_User, time=1, value=0.4}
        ]}}}
     CurveIndex=&RandomRangeFloat(Minimum=0, Maximum=1)
     Scale Curve=2
    )) {isSetParameters=true}
   &SkeletalMeshLocation(Bone / Socket Index=2, Bone Sampling Mode=::NewEnumerator4, MaskSpawnGroup=false, Skeletal Mesh=$System.SkelBody)
   &SphereLocation(
    MaskSpawnGroup=false
    Override Seed=false
    Sphere Radius=35
    Surface Only=true
    UseNonUniformScale=false
    UseOffset=false
    UseSphereOrientation=false
   ) {module=SphereLocation001}
   &AddVelocityFromPoint(Module.Use Falloff Distance=false, Origin Offset=<0 0 -50>, Velocity Strength=15)
   &AddVelocity(Velocity=&Multiply_VectorByFloat(Vector=$Output.SphereLocation001.SphereNormal, Float=20))
   &ApplyInitialForces()
  ]
  particleUpdateScript [
   &ParticleState()
   &CurlNoiseForce(Noise Frequency=4, Noise Strength=75, Pan Noise=true, Pan Noise Field=<0.5 0.5 0>)
   &GravityForce(Gravity=<0 0 555>)
   &ScaleSpriteSize(Scale Factor=&Vector2DFromFloat(Value=&FloatFromCurve(
      FloatCurve=DataInterface() {propertyValues {Curve {keys [<at 0 1 RCIM_Cubic>, {arriveTangent=-1.6315, interpMode=RCIM_Cubic, leaveTangent=-1.6315, tangentMode=RCTM_User, time=1, value=0}]}}}
      Scale Curve=$Particles.RandomCore
     )))
   &ScaleSpriteSize(
    Scale Factor=&Multiply_Vector2D(A=$Particles.SpriteSize, B=&Vector2DFromFloat(Value=0.6))
    Scale Sprite Size Mode=::NewEnumerator2
   ) {module=ScaleSpriteSize002}
   &ScaleColor(
    Scale Alpha=&Multiply_Float(
     A=&FloatFromCurve(FloatCurve=DataInterface() {propertyValues {Curve {keys [<at 0.8 1>,<at 1 0>]}}})
     B=&FNTimeOfDayScalar(Night Scale=0.3, Morning Scale=0.1, Evening Scale=0.2)
    )
    Scale RGB=&VectorFromFloat(Value=&FloatFromCurve(FloatCurve=DataInterface() {propertyValues {Curve {keys [<at 0 1>,<at 0.1 3>,<at 0.2 1>]}}}))
    ScaleA=true
    ScaleRGB=true
   )
   &Drag(Drag=2, Use Linear Drag=true, Use Rotational Drag=false)
   &SolveForcesAndVelocity(Clamp Velocity=false, Limit Acceleration=false)
  ]
  rendererClasses [&NiagaraSpriteRendererProperties]
  renderers [{rendererClass=&NiagaraSpriteRendererProperties}]
 }
 boneflow_radial_peaks_B {
  simTarget=GPUComputeSim
  emitterUpdateScript [
   &EmitterState(Life Cycle Mode=::NewEnumerator1, Loop Behavior=::NewEnumerator1, Loop Duration Mode=::NewEnumerator1, UseLoopDelay=false)
   &SpawnBurst_Instantaneous(Spawn Count=15, Use Spawn Probability=false, UseLoopCountLimit=false)
  ]
  particleSpawnScript [
   &InitializeParticle(
    Color=<rgba 1 0.6 0 0.05>
    Lifetime=&RandomRangeFloat(Minimum=1, Maximum=1)
    Sprite Size Mode=::NewEnumerator3
    Uniform Sprite Size=25
    UsePositionOffset=false
    Write Material Random=true
   )
   /SetVariables(Particles.T:NiagaraFloat=&RemapRange(
     Input Value To Remap=&ReturnNormalizedExecIndex(Normalized Index Includes One=false)
     Input Min=0
     Input Max=1
     Output Min=0
     Output Max=1
    )) {isSetParameters=true, module=SetVariables_#8}
  ]
  particleUpdateScript [
   &ParticleState(Kill Particles When Lifetime Has Elapsed=false)
   &self:Pulse_Rays_01(
    Frequency=3
    Time:NiagaraFloat=$Engine.Time
    Time Offset:NiagaraFloat=&ReturnNormalizedExecIndex(Normalized Index Scale=12)
   ) {module=Pulse_Rays_01}
   &ScaleSpriteSize(
    Non-Uniform Curve Index:NiagaraFloat=&ReturnNormalizedExecIndex()
    Non-Uniform Curve Scale=<3 -2>
    Non-Uniform Curve Sprite Scale=DataInterface() {propertyValues {
      XCurve {keys [
        {arriveTangent=2.5, interpMode=RCIM_Cubic, leaveTangent=2.5, time=0.2, value=-0.025}
        {arriveTangent=3.308, interpMode=RCIM_Cubic, leaveTangent=3.308, tangentMode=RCTM_User, time=0.3, value=0.225}
        {arriveTangent=0.3732, interpMode=RCIM_Cubic, leaveTangent=0.3732, time=0.4, value=0.575}
        {arriveTangent=-0.9433, interpMode=RCIM_Cubic, leaveTangent=-0.9433, time=0.7019, value=0.375}
        {arriveTangent=-1.219, interpMode=RCIM_Cubic, leaveTangent=-1.219, time=1.0095, value=0}
       ]}
      YCurve {keys [
        {arriveTangent=2.5, interpMode=RCIM_Cubic, leaveTangent=2.5, time=0.2, value=-0.025}
        {arriveTangent=3.308, interpMode=RCIM_Cubic, leaveTangent=3.308, tangentMode=RCTM_User, time=0.3, value=0.225}
        {arriveTangent=0.3732, interpMode=RCIM_Cubic, leaveTangent=0.3732, time=0.4, value=0.575}
        {arriveTangent=-0.9023, interpMode=RCIM_Cubic, leaveTangent=-0.9023, time=0.7019, value=0.375}
        {arriveTangent=-1.1377, interpMode=RCIM_Cubic, leaveTangent=-1.1377, time=1.0095, value=0.025}
       ]}
     }}
    Scale Sprite Size Mode=::NewEnumerator3
   ) {module=ScaleSpriteSize003}
   &FN_BoneFlow(
    Bias=0.3515
    Bone Offsets=DataInterface() {propertyValues {
      GpuSyncMode=SyncCpuToGpu
      MaxElements=0
      FloatData [<8 0 0>, <-6 0 0>, <1 0 0>, <6 0 0>, <0 0 0>, <0 0 0>, <0 0 0>, <0 0 0>, <0 0 0>, <0 0 0>, <0 0 0>, <0 0 0>]
     }}
    Bone Sets=DataInterface() {propertyValues {GpuSyncMode=SyncCpuToGpu, MaxElements=0, IntData [0,1,1,2,2,3,3,4]}}
    Enable Position Offsets=true
    Height:NiagaraFloat=$Particles.T
    SkeletalMesh:NiagaraDataInterfaceSkeletalMesh=$System.SkelBody
    Step Size:NiagaraFloat=&Divide_Float(A=0.1, B=1)
    Tension=0.3
    scale=1
   )
   &SampleSkeletalMesh(
    AdditionalYawPitchRoll=false
    Bone / Socket Index=2
    Bone Sampling Mode:ENiagaraBoneSamplingMode=::NewEnumerator4
    "Interpolate Mesh Sampling (Update)"=true
    Position Offset Coordinate Space:ENiagaraCoordinateSpace=::World
    Skeletal Mesh:NiagaraDataInterfaceSkeletalMesh=$System.SkelBody
   )
   &self:ScratchModule_01(
    Offset:Vector3f=<0 0 -25>
    RadialCenter:Vector3f=$Particles.SampleSkeletalMesh.SampledPosition
    offsetScale:NiagaraFloat=&FloatFromCurve(FloatCurve=DataInterface() {propertyValues {Curve {keys [
         {arriveTangent=0.5893, interpMode=RCIM_Cubic, leaveTangent=0.5893, tangentMode=RCTM_User, time=0, value=-0.2}
         {interpMode=RCIM_Cubic, tangentMode=RCTM_User, time=0.5, value=-0.05}
        ]}}})
   ) {module=ScratchModule_01001}
   &CameraOffset()
   &ScaleColor(
    Scale Alpha=&Multiply_Float(
     A=&FloatFromCurve(
      FloatCurve=DataInterface() {propertyValues {Curve {keys [<at 0.8 1>,<at 1 0>]}}}
      CurveIndex=&ReturnNormalizedExecIndex()
     )
     B=&FNTimeOfDayScalar(Night Scale=0.3, Morning Scale=0.1, Evening Scale=0.2)
    )
    Scale RGB=&VectorFromFloat(Value=&FloatFromCurve(
      FloatCurve=DataInterface() {propertyValues {Curve {keys [<at 0 0>,<at 1 1>]}}}
      CurveIndex=&ReturnNormalizedExecIndex()
      Scale Curve=&Lerp_Float(A=1, B=1, Alpha=$Particles.Pulse)
     ))
    ScaleA=true
    ScaleRGB=true
   )
   &self:OffsetZ_01(Offset:Vector3f=<0 0 33>, Pivot:Vector3f=$Particles.SampleSkeletalMesh.SampledPosition, Scale=2) {module=OffsetZ_01}
   &DynamicMaterialParameters(
    Index 0 Param 1=1
    Index 0 Param 2=5
    Index 0 Param 3=0.6
    Index 3 Param 1=1
    Index 3 Param 2=&MakeFloatFromInt(Int=&ReturnExecIndex())
    Module.Param0WriteEnabled=true
    Module.Param0WriteEnabled3=true
    Module.Param1WriteEnabled=true
    Module.Param1WriteEnabled3=true
    Module.Param2WriteEnabled=true
    Module.Param2WriteEnabled3=true
    Module.Param3WriteEnabled=true
    Module.Param3WriteEnabled3=true
    Write Parameter Index 3=true
   )
   &ScaleColor(
    Scale RGB=&VectorFromFloat(Value=&Add_Float(A=1, B=$System.FlareAll))
    ScaleA=false
    ScaleRGB=true
   ) {module=ScaleColor001}
   &ScaleColor(
    Scale Alpha=$User.FXBrightnessScale
    Scale RGB=&VectorFromFloat(Value=$User.FXBrightnessScale)
    ScaleA=true
    ScaleRGB=true
   ) {module=ScaleColor003}
  ]
  rendererClasses [&NiagaraSpriteRendererProperties]
  renderers [{rendererClass=&NiagaraSpriteRendererProperties}]
 }
 boneflow_radial_streaks_body_tilt {
  emitterUpdateScript [
   &EmitterState(Life Cycle Mode=::NewEnumerator1, Loop Behavior=::NewEnumerator1, Loop Duration Mode=::NewEnumerator1, UseLoopDelay=false)
   &SpawnBurst_Instantaneous(Spawn Count=15, Use Spawn Probability=false, UseLoopCountLimit=false)
  ]
  particleSpawnScript [
   &InitializeParticle(
    Color=<rgba 1 0.5 0 0.3>
    Lifetime=&RandomRangeFloat(Minimum=1, Maximum=1)
    Sprite Size Mode=::NewEnumerator3
    Uniform Sprite Size=25
    UsePositionOffset=false
    Write Material Random=true
   )
   /SetVariables(Particles.T:NiagaraFloat=&RemapRange(
     Input Value To Remap=&ReturnNormalizedExecIndex(Normalized Index Includes One=false)
     Input Min=0
     Input Max=1
     Output Min=0
     Output Max=1
    )) {isSetParameters=true, module=SetVariables_#8}
  ]
  particleUpdateScript [
   &ParticleState(Kill Particles When Lifetime Has Elapsed=false)
   &self:Pulse_Rays_01(
    Frequency=12
    Time:NiagaraFloat=$Engine.Time
    Time Offset:NiagaraFloat=&ReturnNormalizedExecIndex(Normalized Index Scale=12)
   ) {module=Pulse_Rays_01}
   &ScaleSpriteSize(
    Non-Uniform Curve Index:NiagaraFloat=&ReturnNormalizedExecIndex()
    Non-Uniform Curve Scale=<0.75 -6>
    Non-Uniform Curve Sprite Scale=DataInterface() {propertyValues {
      XCurve {keys [
        {arriveTangent=0.3555, interpMode=RCIM_Cubic, leaveTangent=0.3555, tangentMode=RCTM_User, time=0.2178, value=0.0429}
        {arriveTangent=0.3309, interpMode=RCIM_Cubic, leaveTangent=0.3309, tangentMode=RCTM_User, time=0.3583, value=0.4179}
        {arriveTangent=0.3661, interpMode=RCIM_Cubic, leaveTangent=0.3661, time=0.4649, value=0.4949}
        {arriveTangent=-1.2495, interpMode=RCIM_Cubic, leaveTangent=-1.2495, time=0.6856, value=0.5378}
        {arriveTangent=-0.0225, interpMode=RCIM_Cubic, leaveTangent=-0.0225, tangentMode=RCTM_User, time=0.9009, value=-0.05}
       ]}
      YCurve {keys [
        {arriveTangent=0.3555, interpMode=RCIM_Cubic, leaveTangent=0.3555, tangentMode=RCTM_User, time=0.2178, value=0.0429}
        {arriveTangent=0.3309, interpMode=RCIM_Cubic, leaveTangent=0.3309, tangentMode=RCTM_User, time=0.3583, value=0.4179}
        {arriveTangent=0.3661, interpMode=RCIM_Cubic, leaveTangent=0.3661, time=0.4649, value=0.4949}
        {arriveTangent=-1.1348, interpMode=RCIM_Cubic, leaveTangent=-1.1348, time=0.6856, value=0.5378}
        {arriveTangent=0.055, interpMode=RCIM_Cubic, leaveTangent=0.055, tangentMode=RCTM_User, time=0.9009, value=0}
       ]}
     }}
    Scale Sprite Size Mode=::NewEnumerator3
   ) {module=ScaleSpriteSize003}
   &FN_BoneFlow(
    Bias=0.3515
    Bone Offsets=DataInterface() {propertyValues {
      GpuSyncMode=SyncCpuToGpu
      MaxElements=0
      FloatData [<3 0 0>, <-2 0 0>, <1 0 0>, <1 0 0>, <3 0 0>, <0 0 0>, <0 0 0>, <0 0 0>, <0 0 0>, <0 0 0>, <0 0 0>, <0 0 0>]
     }}
    Bone Sets=DataInterface() {propertyValues {GpuSyncMode=SyncCpuToGpu, MaxElements=0, IntData [0,1,1,2,2,3,3,4]}}
    Enable Position Offsets=true
    Height:NiagaraFloat=$Particles.T
    SkeletalMesh:NiagaraDataInterfaceSkeletalMesh=$System.SkelBody
    Step Size:NiagaraFloat=&Divide_Float(A=0.1, B=1)
    Tension=0.3
    scale=1
   ) {module=FN_BoneFlow001}
   &SampleSkeletalMesh(
    AdditionalYawPitchRoll=false
    Bone / Socket Index=2
    Bone Sampling Mode:ENiagaraBoneSamplingMode=::NewEnumerator4
    "Interpolate Mesh Sampling (Update)"=true
    Position Offset Coordinate Space:ENiagaraCoordinateSpace=::World
    Skeletal Mesh:NiagaraDataInterfaceSkeletalMesh=$System.SkelBody
   )
   &self:SampleAlignment(Bone_B=1, SkeletalMesh:NiagaraDataInterfaceSkeletalMesh=$System.SkelMesh_Align) {module=SampleAlignment}
   &self:ScratchModule_01(
    Offset:Vector3f=<0 0 -25>
    RadialCenter:Vector3f=&Add_Vector(
     A=$Particles.SampleSkeletalMesh.SampledPosition
     B=&Multiply_VectorByFloat(Vector=$Particles.Direction, Float=-15)
    )
    offsetScale:NiagaraFloat=&FloatFromCurve(FloatCurve=DataInterface() {propertyValues {Curve {keys [
         {arriveTangent=0.5893, interpMode=RCIM_Cubic, leaveTangent=0.5893, tangentMode=RCTM_User, time=0, value=-0.2}
         {interpMode=RCIM_Cubic, tangentMode=RCTM_User, time=0.5, value=-0.05}
        ]}}})
   ) {module=ScratchModule_01001}
   &CameraOffset(Camera Offset Amount=8)
   &ScaleColor(
    Scale Alpha=&Multiply_Float(
     A=&FloatFromCurve(
      FloatCurve=DataInterface() {propertyValues {Curve {keys [<at 0.8 1>,<at 1 0>]}}}
      CurveIndex=&ReturnNormalizedExecIndex()
     )
     B=&FNTimeOfDayScalar(Night Scale=0.3, Morning Scale=0.1, Evening Scale=0.2)
    )
    Scale RGB=&VectorFromFloat(Value=&FloatFromCurve(
      FloatCurve=DataInterface() {propertyValues {Curve {keys [<at 0 0>,<at 1 1>]}}}
      CurveIndex=&ReturnNormalizedExecIndex(Normalized Index Scale=1.3)
      Scale Curve=&Lerp_Float(A=1, B=3, Alpha=$Particles.Pulse)
     ))
    ScaleA=true
    ScaleRGB=true
   )
   &ScaleColor(
    Scale Alpha=&FloatFromCurve(FloatCurve=DataInterface() {propertyValues {Curve {keys [
         {arriveTangent=9.0725, interpMode=RCIM_Cubic, leaveTangent=9.0725, tangentMode=RCTM_User, time=0, value=1}
         {arriveTangent=-16.4406, interpMode=RCIM_Cubic, leaveTangent=-16.4406, tangentMode=RCTM_User, time=0.1408, value=5}
         {arriveTangent=-0.0405, interpMode=RCIM_Cubic, leaveTangent=-0.0405, tangentMode=RCTM_User, time=0.7, value=1.3}
        ]}}})
    ScaleA=true
    ScaleRGB=false
   ) {module=ScaleColor002}
   &self:OffsetZ_01(
    Offset:Vector3f=&Add_Vector(A=<0 0 33>, B=&Multiply_VectorByFloat(Vector=$Particles.Direction, Float=35))
    Pivot:Vector3f=$Particles.SampleSkeletalMesh.SampledPosition
    Scale=2
   ) {module=OffsetZ_01001}
   &DynamicMaterialParameters(
    Index 3 Param 1=$System.FlareAll
    Module.Param0WriteEnabled3=true
    Module.Param1WriteEnabled3=true
    Module.Param2WriteEnabled3=true
    Module.Param3WriteEnabled3=true
    Write Parameter Index 0=false
    Write Parameter Index 3=true
   )
   &ScaleColor(
    Scale RGB=&VectorFromFloat(Value=&Add_Float(A=1, B=$System.FlareAll))
    ScaleA=false
    ScaleRGB=true
   ) {module=ScaleColor001}
   &ScaleColor(
    Scale Alpha=$User.FXBrightnessScale
    Scale RGB=&VectorFromFloat(Value=$User.FXBrightnessScale)
    ScaleA=true
    ScaleRGB=true
   ) {module=ScaleColor003}
   &ScaleColor(Scale Alpha=$System.CameraDistanceAlpha, ScaleA=true, ScaleRGB=false) {module=ScaleColor004}
  ]
  rendererClasses [&NiagaraSpriteRendererProperties]
  renderers [
   {rendererClass=&NiagaraSpriteRendererProperties}
   {rendererClass=&NiagaraSpriteRendererProperties, rendererIndex=1}
  ]
 }
 rays_mesh {
  emitterUpdateScript [
   &EmitterState(Inactive Response=::NewEnumerator2, Life Cycle Mode=::NewEnumerator1, Loop Behavior=::NewEnumerator1, Loop Duration=1, UseLoopDelay=false)
   &SpawnBurst_Instantaneous(Use Spawn Probability=false, UseLoopCountLimit=false)
  ]
  particleSpawnScript [&InitializeParticle(
    Color=<rgba 1 0.4 0 0.3>
    Lifetime=125
    Sprite Size Mode=::NewEnumerator3
    Uniform Sprite Size=75
    UsePositionOffset=false
    Write Material Random=true
   )]
  particleUpdateScript [
   &SkeletalMeshLocation(
    Bone Sampling Mode=::NewEnumerator4
    MaskSpawnGroup=false
    Orientation Sampling=::NewEnumerator1
    Sampled Position Offset=<0 0 15>
    Sampled Position Offset Coordinate Space:ENiagaraSkelMeshTransforms=::NewEnumerator1
    Skeletal Mesh=$System.SkelMesh_Chest
   ) {module=SkeletalMeshLocation001}
   &CameraOffset(Camera Offset Amount=-10)
   /SetVariables(Particles.Scale=<4 4 8>) {isSetParameters=true, module=SetVariables_#9}
   /SetVariables(Particles.SpriteSize=<50 60>) {isSetParameters=true, module=SetVariables_#10}
   &DynamicMaterialParameters(
    Index 0 Param 1=&Multiply_Float(A=&MakeFloatFromInt(Int=&ReturnExecIndex()), B=22)
    Index 0 Param 2=1
    Index 0 Param 3=5
    Index 0 Param 4=5
    Module.Param0WriteEnabled=true
    Module.Param1WriteEnabled=false
    Module.Param2WriteEnabled=true
    Module.Param3WriteEnabled=false
   )
   /SetVariables(Particles.ColorB:LinearColor=<rgba 1 0.5 0 1>) {isSetParameters=true}
   &ScaleColor(
    Scale RGB=&VectorFromFloat(Value=&Add_Float(A=1, B=$System.FlareAll))
    ScaleA=false
    ScaleRGB=true
   )
   &ScaleColor(
    Scale Alpha=$User.FXBrightnessScale
    Scale RGB=&VectorFromFloat(Value=$User.FXBrightnessScale)
    ScaleA=true
    ScaleRGB=true
   ) {module=ScaleColor003}
  ]
  rendererClasses [&NiagaraMeshRendererProperties]
  renderers [{rendererClass=&NiagaraMeshRendererProperties, rendererIndex=0}]
 }
 rays_mesh_purpler {
  simTarget=GPUComputeSim
  emitterUpdateScript [
   &EmitterState(Inactive Response=::NewEnumerator2, Life Cycle Mode=::NewEnumerator1, Loop Behavior=::NewEnumerator1, Loop Duration=1, UseLoopDelay=false)
   &SpawnBurst_Instantaneous(Use Spawn Probability=false, UseLoopCountLimit=false)
  ]
  particleSpawnScript [&InitializeParticle(
    Color=<rgba 0.0833 0.0056 0.0485 0.3>
    Lifetime=125
    Sprite Size Mode=::NewEnumerator3
    Uniform Sprite Size=75
    UsePositionOffset=false
    Write Material Random=true
   )]
  particleUpdateScript [
   &SkeletalMeshLocation(
    Bone Sampling Mode=::NewEnumerator4
    MaskSpawnGroup=false
    Orientation Sampling=::NewEnumerator1
    Sampled Position Offset=<0 0 15>
    Sampled Position Offset Coordinate Space:ENiagaraSkelMeshTransforms=::NewEnumerator1
    Skeletal Mesh=$System.SkelMesh_Chest
   ) {module=SkeletalMeshLocation001}
   &CameraOffset(Camera Offset Amount=-5)
   /SetVariables(Particles.Scale=<4 4 8>) {isSetParameters=true, module=SetVariables_#9}
   /SetVariables(Particles.SpriteSize=<50 60>) {isSetParameters=true, module=SetVariables_#10}
   &DynamicMaterialParameters(
    Index 0 Param 1=&Multiply_Float(A=&MakeFloatFromInt(Int=&ReturnExecIndex()), B=22)
    Index 0 Param 2=1
    Index 0 Param 3=5
    Index 0 Param 4=5
    Module.Param0WriteEnabled=true
    Module.Param1WriteEnabled=false
    Module.Param2WriteEnabled=true
    Module.Param3WriteEnabled=false
   )
   &ScaleColor(
    Scale RGB=&VectorFromFloat(Value=&Add_Float(A=0.3, B=$System.FlareAll))
    ScaleA=false
    ScaleRGB=true
   )
   &ScaleColor(
    Scale Alpha=$User.FXBrightnessScale
    Scale RGB=&VectorFromFloat(Value=$User.FXBrightnessScale)
    ScaleA=true
    ScaleRGB=true
   ) {module=ScaleColor003}
  ]
  rendererClasses [&NiagaraMeshRendererProperties]
  renderers [{rendererClass=&NiagaraMeshRendererProperties, rendererIndex=0}]
 }
}
summary {
 systemName=NS_VC_body_B
 emitters [
  {emitterName=rays_mesh, rendererClasses [&NiagaraMeshRendererProperties]}
  {emitterName=boneflow_radial_peaks_B, simTarget=GPUComputeSim, rendererClasses [&NiagaraSpriteRendererProperties]}
  {emitterName=rays_mesh_purpler, simTarget=GPUComputeSim, rendererClasses [&NiagaraMeshRendererProperties]}
  {emitterName=Sparckle_Body, rendererClasses [&NiagaraSpriteRendererProperties]}
  {emitterName=boneflow_radial_streaks_body_tilt, rendererClasses [&NiagaraSpriteRendererProperties]}
 ]
 userVariables [
  User.bIsFrontEndPreview {type=NiagaraBool}
  User.Flare {type=NiagaraFloat, defaultValue=0, description=""}
  User.FXBrightnessScale {type=NiagaraFloat, defaultValue=1, description=""}
 ]
}
